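# Translucent unlit trail material for the Calavera backpack, applied to Niagara mesh particles.
# Emissive layers distorted noise, caustics, rainbow detail and a dissolve edge glow, tinted by particle color.
# Opacity combines curve-atlas or smoke-trail masks with noise dissolve, a particle-centered sphere mask and depth fade.
# World position offset sums curve-shaped SplineThicken, Perlin noise waves and a clamped interpolation offset, with Niagara dynamic parameters driving width, dissolve and motion.
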Material /Game/Accessories/FORT_Backpacks/Backpack_M_MED_Calavera/FX/M_Backpack_Calavera_Trail
graph {
 properties {
  bUsedWithNiagaraMeshParticles=true
  blendMode=BLEND_Translucent
  pixelDepthOffsetMode=PDOM_Legacy
  shadingModel=MSM_Unlit
  twoSided=true
  usedShadingModels=Unlit
 }
 nodes [
  n1 = TextureCoordinate()
  n2 = VectorParameter() {group=Emissive, parameterName=Emissive HighLight UV, sortPriority=1, defaultValue=<rgba 0.1 0.25 0.25 0.25>}
  n3 = Multiply(
   A=AppendVector(A=@n2.R, B=@n2.R)
   B=AppendVector(
    A=ScalarParameter() {defaultValue=1, group=Noise, parameterName=Global Scale U}
    B=ScalarParameter() {defaultValue=1, group=Noise, parameterName=Global Scale V}
   )
  )
  n4 = Add(A=AppendVector(A=ParticleRandom(), B=Constant()), B=Panner(Coordinate=Multiply(A=@n1, B=@n3), Speed=AppendVector(A=@n2.B, B=@n2.B)))
  MaterialFunctionCall(UVs=@n4, "Rotation Angle (0-1)"=Constant() {r=0.25}) {materialFunction=&CustomRotator}
  n5 = Desaturation(
   None=VectorParameter() {group=Emissive, parameterName=Emissive Color, defaultValue=<rgba 1 1 1 1>}.RGB
   Fraction=ScalarParameter() {group=Emissive, parameterName=Desaturation Color}
  )
  n6 = Multiply(
   A=Multiply(
    A=TextureSample(UVs=@n4) {texture=&T_ShieldCaustics}.R
    B=ScalarParameter() {defaultValue=10, group=Emissive, parameterName=Emissive HighLight Intensity}
   )
   B=@n5
  )
  n7 = ParticleRandom()
  n8 = AppendVector(A=@n7, B=@n7)
  n9 = VectorParameter() {group=Emissive, parameterName=Emissive Detail UV, sortPriority=1, defaultValue=<rgba 0.5 0.5 0 0.2>}
  n10 = AppendVector(A=@n9.R, B=@n9.R)
  n11 = AppendVector(
   A=ScalarParameter() {defaultValue=1, group=Noise, parameterName=Global Scale U}
   B=ScalarParameter() {defaultValue=1, group=Noise, parameterName=Global Scale V}
  )
  n12 = Panner(Coordinate=Multiply(A=TextureCoordinate(), B=Multiply(A=@n10, B=@n11)), Speed=AppendVector(A=@n9.B, B=@n9.B))
  n13 = ParticleRandom()
  n14 = VectorParameter() {group=Distortion, parameterName=Distortion UV, sortPriority=1, defaultValue=<rgba 1 1 0 0>}
  n15 = AppendVector(
   A=ScalarParameter() {defaultValue=1, group=Noise, parameterName=Global Scale U}
   B=ScalarParameter() {defaultValue=1, group=Noise, parameterName=Global Scale V}
  )
  n16 = Multiply(A=TextureCoordinate(), B=Multiply(A=AppendVector(A=@n14.R, B=@n14.R), B=@n15))
  n17 = AppendVector(A=@n14.B, B=@n14.B)
  n18 = TextureSampleParameter2D(UVs=Add(A=AppendVector(A=@n13, B=@n13), B=Panner(Coordinate=@n16, Speed=@n17))) {
   group=Distortion
   parameterName="Distortion  Tex"
   samplerType=SAMPLERTYPE_Normal
   sortPriority=0
   texture=&T_Weap_Distort_N_02
  }
  n19 = Multiply(
   A=ComponentMask(None=@n18.RGB) {g=true, r=true}
   B=ScalarParameter() {defaultValue=0.2, group=Distortion, parameterName=Distortion Intensity, sortPriority=2}
  )
  n20 = TextureSample(UVs=Add(A=Add(A=@n8, B=@n12), B=@n19)) {texture=&T_Tile_RainbowColor}
  n21 = ScalarParameter() {group=Emissive, parameterName=Emissive Detail Intensity}
  n22 = ParticleRandom()
  n23 = AppendVector(A=@n22, B=@n22)
  n24 = VectorParameter() {group=Noise, parameterName=Noise A UV, sortPriority=6, defaultValue=<rgba 1 0.5 -0.2 0.7>}
  n25 = AppendVector(A=@n24.R, B=@n24.R)
  n26 = Panner(Coordinate=Multiply(A=TextureCoordinate(), B=Multiply(A=@n25, B=@n15)), Speed=AppendVector(A=@n24.B, B=@n24.B))
  n27 = TextureSampleParameter2D(UVs=Add(A=@n19, B=Add(A=@n23, B=@n26))) {group=Noise, parameterName=Noise Tex, samplerType=SAMPLERTYPE_LinearColor, texture=/Game/Global/Textures/Default/Noise/T_Noise01}
  n28 = CurveAtlasRowParameter(CurveTime=ComponentMask(None=TextureCoordinate()) {r=true}) {
   atlas=/Game/Athena/LinearColorCurves/A_Atlas_Turtleneck_Glider
   curve=&/Curve_Backpack_Calavera_02
   defaultValue=40
   group=Emissive
   parameterName=Curve Color
  }
  n29 = CurveAtlasRowParameter(CurveTime=ComponentMask(None=TextureCoordinate()) {g=true}) {
   atlas=&A_Atlas_Turtleneck_Glider
   curve=&Curve_Backpack_Calavera_02
   defaultValue=40
   group=Emissive
   parameterName=Curve Color
  }
  n30 = Multiply(
   A=LinearInterpolate(
    A=ScalarParameter() {defaultValue=0.2, group=Noise, parameterName=Noise A Dark}
    B=ScalarParameter() {defaultValue=3, group=Noise, parameterName=Noise A Light}
    Alpha=Power(Base=@n27.R)
   )
   B=Multiply(A=@n28.G, B=@n29.R)
  )
  n31 = ScalarParameter() {defaultValue=100, group=Dissolve, parameterName=Dissolve Edge Intensity}
  n32 = DynamicParameter() {defaultValue=<rgba 0 1 1 1>, paramNames [Random Offset, Spline Width Mult, Dissolve, Dissolve Edge Intensity]}
  n33 = TextureCoordinate()
  n34 = TextureCoordinate()
  n35 = Multiply(A=ComponentMask(None=@n34) {r=true}, B=OneMinus(None=ComponentMask(None=@n34) {r=true}))
  n36 = Multiply(
   A=LinearInterpolate(B=Multiply(A=@n35) {constB=2}, Alpha=ScalarParameter() {defaultValue=0.5, group=Dissolve, parameterName=Lerp Gradient Tip}) {constA=1, constAlpha=1}
   B=ComponentMask(None=TextureCoordinate()) {g=true}
  )
  n37 = ParticleRandom()
  n38 = VectorParameter() {group=Dissolve, parameterName=Dissolve Tex UV, sortPriority=6, defaultValue=<rgba 0.5 0.1 0 0>}
  n39 = Multiply(A=TextureCoordinate(), B=AppendVector(A=@n38.R, B=@n38.R))
  n40 = AppendVector(A=@n38.B, B=@n38.B)
  n41 = TextureSampleParameter2D(UVs=Add(A=AppendVector(A=@n37, B=@n37), B=Panner(Coordinate=@n39, Speed=@n40))) {group=Dissolve, parameterName=Dissolve Tex, texture=/Game/Global/Textures/Default/Noise/T_FireNoise_01}
  n42 = Power(
   Base=LinearInterpolate(A=@n36, B=Multiply(A=@n36, B=@n41.R), Alpha=ScalarParameter() {defaultValue=0.5, group=Dissolve, parameterName=Lerp Dissolve Tex})
   Exp=ScalarParameter() {defaultValue=1, group=Dissolve, parameterName=Dissolve Power}
  )
  n43 = ScalarParameter() {defaultValue=4, group=Dissolve, parameterName=Dissolve Edge Hardness, sliderMin=0.5}
  n44 = DynamicParameter() {defaultValue=<rgba 0 1 1 1>, paramNames [Random Offset, Spline Width Mult, Dissolve, Dissolve Edge Intensity]}
  n45 = Saturate(None=Subtract(A=Multiply(A=@n42, B=@n43), B=LinearInterpolate(A=@n43, Alpha=@n44.Dissolve) {constB=-1}))
  n46 = Multiply(A=Multiply(A=Multiply(A=@n31, B=@n32.Dissolve Edge Intensity), B=ComponentMask(None=@n33) {g=true}), B=Subtract(A=Constant() {r=1}, B=@n45))
  n47 = Add(A=Multiply(A=Multiply(A=@n20.RGB, B=@n21), B=@n5), B=Add(A=Multiply(A=@n30, B=@n5), B=Multiply(A=@n5, B=@n46)))
  n48 = ParticleColor()
  MaterialFunctionCall(Emissive=Multiply(A=Add(A=@n6, B=@n47), B=@n48.RGB)) {materialFunction=&MF_FX_EmissiveAdjustments}
  n49 = ComponentMask(None=TextureCoordinate()) {g=true}
  n50 = CurveAtlasRowParameter(CurveTime=@n49) {
   atlas=&A_Atlas_Turtleneck_Glider
   curve=&Curve_Backpack_Calavera_02
   defaultValue=40
   group=Emissive
   parameterName=Curve Color
  }
  n51 = CurveAtlasRowParameter(CurveTime=@n49) {
   atlas=&A_Atlas_Turtleneck_Glider
   curve=&Curve_Backpack_Calavera_02
   defaultValue=40
   group=Emissive
   parameterName=Curve Color
  }
  n52 = ScalarParameter() {defaultValue=1, group=Opacity, parameterName=Lerp Alpha Curve}
  n53 = Constant()
  n54 = ParticleRandom()
  n55 = VectorParameter() {group=Opacity, parameterName=SmokeTrail UV, sortPriority=1, defaultValue=<rgba 1 1 0 0.2>}
  n56 = ScalarParameter() {defaultValue=1, group=Noise, parameterName=Global Scale U}
  n57 = ScalarParameter() {defaultValue=1, group=Noise, parameterName=Global Scale V}
  n58 = Multiply(A=TextureCoordinate(), B=Multiply(A=AppendVector(A=@n55.R, B=@n55.R), B=AppendVector(A=@n56, B=@n57)))
  n59 = AppendVector(A=@n55.B, B=@n55.B)
  n60 = Add(A=Add(A=AppendVector(A=@n53, B=@n54), B=Panner(Coordinate=@n58, Speed=@n59)), B=@n19)
  n61 = Constant() {r=0.25}
  n62 = MaterialFunctionCall(
   In=TextureSampleParameter2D(UVs=MaterialFunctionCall(UVs=@n60, "Rotation Angle (0-1)"=@n61) {materialFunction=&CustomRotator}.Rotated Values) {parameterName=SmokeTrailTexture, texture=/Game/Effects/Fort_Effects/Textures/Smoke/T_Smoke_Trail_Soft_Chromatic}.R
   Contrast=ScalarParameter() {group=Opacity, parameterName=SmokeTrailContrast}
  ) {materialFunction=&CheapContrast}
  n63 = ComponentMask(None=TextureCoordinate()) {g=true}
  n64 = Multiply(A=Power(Base=OneMinus(None=@n63))) {constB=6}
  n65 = Power(Base=ComponentMask(None=TextureCoordinate()) {g=true}, Exp=ScalarParameter() {defaultValue=1, group=Opacity, parameterName=TailFadePower}) {constExponent=1}
  n66 = Multiply(
   A=LinearInterpolate(A=Constant2Vector() {g=1, r=1}, B=Saturate(None=@n64), Alpha=ScalarParameter() {defaultValue=1, group=Opacity, parameterName=UseHeadFade})
   B=Saturate(None=Multiply(A=@n65) {constB=3})
  )
  n67 = Multiply(
   A=ScalarParameter() {defaultValue=1, group=Opacity, parameterName=Opactiy Mult}
   B=StaticSwitchParameter(True=Multiply(A=@n62.Result, B=@n66), False=LinearInterpolate(A=@n50.B, B=@n51.A, Alpha=@n52)) {group=Opacity, parameterName=Use SmokeTrail Mask}
  )
  n68 = OneMinus(None=ComponentMask(None=TextureCoordinate()) {g=true})
  n69 = Saturate(None=Multiply(A=Power(Base=@n68)) {constB=10})
  n70 = WorldPosition()
  n71 = ParticlePositionWS()
  n72 = ScalarParameter() {defaultValue=50, group=Opacity, parameterName=SphereMask Range}
  n73 = ScalarParameter() {defaultValue=0.2, group=Opacity, parameterName=SphereMask Hardness}
  n74 = Multiply(
   A=Multiply(A=Multiply(A=@n67, B=@n69), B=@n45)
   B=StaticSwitchParameter(True=SphereMask(A=@n70.XYZ, B=@n71, Radius=@n72, Hardness=@n73) {attenuationRadius=50, hardnessPercent=0.2}, False=Constant() {r=1}) {defaultValue=true, group=Opacity, parameterName=Use SphereMask}
  )
  n75 = ScalarParameter() {defaultValue=2, group=Opacity, parameterName=DepthFadeDistance}
  Multiply(A=@n48.A, B=ShadingPathSwitch(Default=DepthFade(Opacity=@n74, FadeDistance=@n75), Mobile=@n74) {desc=Migrated from FeatureLevelSwitch})
  n76 = DynamicParameter() {
   desc=Parameters set by Niagara
   parameterIndex=3
   defaultValue=<rgba 0 1 1 1>
   paramNames [ShapeTypeShift, Spline Width Mult, Dissolve, Dissolve Edge Intensity]
  }
  n77 = MaterialFunctionCall(Float3=Subtract(
    A=MaterialFunctionCall(X=@n76.ShapeTypeShift, Y=@n76.ShapeTypeShift, Z=@n76.ShapeTypeShift) {materialFunction=&MakeFloat3}.Result
    B=ParticlePositionWS()
   )) {materialFunction=&BreakOutFloat3Components}
  n78 = ScalarParameter() {defaultValue=100, group=Interpolate, parameterName=Interpolate Clamp XY}
  n79 = Clamp(None=AppendVector(A=@n77.R, B=@n77.R), Min=Multiply(A=@n78) {constB=-1}, Max=@n78) {maxDefault=1000, minDefault=-1000}
  n80 = ScalarParameter() {defaultValue=20, group=Interpolate, parameterName=Interpolate Clamp Z}
  n81 = Clamp(None=@n77.B, Min=Multiply(A=@n80) {constB=-1}, Max=@n80) {minDefault=-1}
  n82 = TextureCoordinate()
  n83 = OneMinus(None=@n82)
  n84 = Multiply(A=ComponentMask(None=OneMinus(None=@n82)) {g=true}, B=ComponentMask(None=@n82) {g=true})
  n85 = ScalarParameter() {group=Interpolate, parameterName=Lerp Interpolate Mask Type}
  n86 = DynamicParameter() {parameterIndex=1, defaultValue=<rgba 0 10 1 1>, paramNames [Random Offset, Noise Intensty, Dissolve, Dissolve Edge Intensity]}
  n87 = ScalarParameter() {group=SplineThicken, parameterName=UseDynamicParameterForShapeType}
  n88 = Power(
   Base=LinearInterpolate(A=ComponentMask(None=@n83) {g=true}, B=Multiply(A=@n84) {constB=1.5}, Alpha=LinearInterpolate(A=@n85, B=@n86.Random Offset, Alpha=@n87))
   Exp=ScalarParameter() {defaultValue=1.7, group=Interpolate, parameterName=Interpolate Power}
  )
  n89 = StaticSwitchParameter(True=LinearInterpolate(B=AppendVector(A=@n79, B=@n81), Alpha=Multiply(A=@n76.Dissolve Edge Intensity, B=@n88)), False=Constant()) {defaultValue=true, group=Interpolate, parameterName=Use Interpolate WPO, sortPriority=0}
  n90 = DynamicParameter() {defaultValue=<rgba 0 1 1 1>, paramNames [Random Offset, Spline Width Mult, Dissolve, Dissolve Edge Intensity]}
  n91 = Multiply(
   A=MaterialFunctionCall(
    WidthBase=ScalarParameter() {defaultValue=1, group=SplineThicken, parameterName=SplineWidthBase, sortPriority=2}
    WidthTip=ScalarParameter() {defaultValue=1, group=SplineThicken, parameterName=SplineWidthTip, sortPriority=1}
   ) {materialFunction=&SplineThicken}.WorldPositionOffset
   B=Multiply(A=ScalarParameter() {defaultValue=1, group=SplineThicken, parameterName=Spline Width Mult}, B=@n90.Spline Width Mult)
  )
  n92 = ScalarParameter() {defaultValue=1, group=SplineThicken, parameterName=Spline Width Mult 02}
  n93 = ComponentMask(None=TextureCoordinate()) {g=true}
  n94 = CurveAtlasRowParameter(CurveTime=@n93) {
   atlas=&A_Atlas_Turtleneck_Glider
   curve=&/Curve_Backpack_Calavera_01
   defaultValue=39
   group=SplineThicken
   parameterName=Curve Shape
  }
  n95 = CurveAtlasRowParameter(CurveTime=@n93) {
   atlas=&A_Atlas_Turtleneck_Glider
   curve=&Curve_Backpack_Calavera_01
   defaultValue=39
   group=SplineThicken
   parameterName=Curve Shape
  }
  n96 = ScalarParameter() {group=SplineThicken, parameterName=Lerp Shape Type}
  n97 = StaticSwitchParameter(True=Multiply(A=Multiply(A=@n91, B=@n92), B=LinearInterpolate(A=@n94.A, B=@n95.B, Alpha=@n96)), False=Constant()) {defaultValue=true, group=SplineThicken, parameterName=Use SplineThicken, sortPriority=0}
  n98 = Constant()
  n99 = DynamicParameter() {defaultValue=<rgba 0 1 1 1>, paramNames [Random Offset, Spline Width Mult, Dissolve, Dissolve Edge Intensity]}
  n100 = VectorParameter() {
   group=WPO
   parameterName=WPO Depth UV
   sortPriority=6
   channelNames {a=WPO Depth V Pan, b=WPO Depth U Pan, g=WPO Depth V, r=WPO Depth U}
   defaultValue=<rgba 0 0.5 0 0.25>
  }
  n101 = Multiply(
   A=AppendVector(A=@n100.WPO Depth U, B=@n100.WPO Depth U)
   B=AppendVector(
    A=ScalarParameter() {defaultValue=1, group=Noise, parameterName=Global Scale U}
    B=ScalarParameter() {defaultValue=1, group=Noise, parameterName=Global Scale V}
   )
  )
  n102 = TextureCoordinate()
  n103 = Add(
   A=AppendVector(A=@n99.Random Offset, B=@n99.Random Offset)
   B=Panner(Coordinate=Multiply(A=@n102, B=@n101), Speed=AppendVector(A=@n100.WPO Depth U Pan, B=@n100.WPO Depth U Pan))
  )
  n104 = Multiply(
   A=LinearInterpolate(Alpha=TextureSample(UVs=@n103) {samplerType=SAMPLERTYPE_LinearColor, texture=/Game/Global/Textures/Default/Tileable_Patterns/T_PerlinNoise_Uncompressed}.R) {constB=-1}
   B=ScalarParameter() {defaultValue=4, group=WPO, parameterName=WPO Depth Intensity}
  ) {constB=6}
  n105 = ComponentMask(None=TextureCoordinate()) {g=true}
  n106 = Multiply(A=Power(Base=OneMinus(None=@n105), Exp=Constant() {r=2}) {constExponent=0.5}, B=Power(Base=@n105, Exp=Constant() {r=1}) {constExponent=0.5})
  n107 = Multiply(
   A=Constant() {r=6}
   B=ScalarParameter() {defaultValue=1, group=WPO, parameterName=WPO Wave Intensity, sliderMax=10, sliderMin=1, sortPriority=33}
  )
  n108 = DynamicParameter() {defaultValue=<rgba 0 1 1 1>, paramNames [Random Offset, Spline Width Mult, Dissolve, Dissolve Edge Intensity]}
  n109 = VectorParameter() {group=WPO, parameterName=WPO Noise UV, sortPriority=6, defaultValue=<rgba 0 0.65 0 0.4>}
  n110 = ScalarParameter() {defaultValue=1, group=Noise, parameterName=Global Scale U}
  n111 = ScalarParameter() {defaultValue=1, group=Noise, parameterName=Global Scale V}
  n112 = TextureCoordinate()
  n113 = ComponentMask(None=@n112) {g=true}
  n114 = Multiply(
   A=AppendVector(A=ComponentMask(None=@n112) {r=true}, B=Power(Base=@n113) {constExponent=1})
   B=Multiply(A=AppendVector(A=@n109.R, B=@n109.R), B=AppendVector(A=@n110, B=@n111))
  )
  n115 = AppendVector(A=@n109.B, B=@n109.B)
  n116 = TextureSampleParameter2D(UVs=Add(A=AppendVector(A=@n108.Random Offset, B=@n108.Random Offset), B=Panner(Coordinate=@n114, Speed=@n115))) {
   parameterName=WPO Noise texture
   samplerType=SAMPLERTYPE_LinearColor
   texture=&T_PerlinNoise_Uncompressed
  }
  n117 = LinearInterpolate(Alpha=@n116.R) {constA=-1}
  n118 = MaterialFunctionCall(R=@n117, G=Constant(), B=Multiply(A=@n117) {constB=0.5}) {materialFunction=&AppendMany}
  n119 = DynamicParameter() {parameterIndex=1, defaultValue=<rgba 0 10 1 1>, paramNames [Random Offset, Noise Intensty, Dissolve, Dissolve Edge Intensity]}
  n120 = LinearInterpolate(
   A=ScalarParameter() {defaultValue=5, group=WPO, parameterName=WPO Noise Intensity}
   B=@n119.Noise Intensty
   Alpha=ScalarParameter() {parameterName=UseDynNoiseIntensity}
  )
  n121 = ComponentMask(None=TextureCoordinate()) {g=true}
  n122 = Constant() {r=0.9}
  n123 = Multiply(
   A=Power(Base=OneMinus(None=@n121), Exp=Constant() {r=1.8}) {constExponent=0.5}
   B=LinearInterpolate(B=Power(Base=@n121, Exp=@n122) {constExponent=0.5}, Alpha=ScalarParameter() {group=SplineThicken, parameterName=WPO Top Mask}) {constA=1}
  )
  n124 = CurveAtlasRowParameter(CurveTime=ComponentMask(None=TextureCoordinate()) {g=true}) {
   atlas=&A_Atlas_Turtleneck_Glider
   curve=&Curve_Backpack_Calavera_01
   defaultValue=39
   group=SplineThicken
   parameterName=Curve Shape
  }
  n125 = ScalarParameter() {group=WPO, parameterName=Lerp WPO Curve}
  n126 = Add(
   A=Multiply(
    A=MaterialFunctionCall(R=@n98, G=@n104, B=@n98) {materialFunction=&AppendMany}.RGB
    B=Multiply(A=@n106, B=@n107) {constB=4}
   )
   B=Multiply(A=Multiply(A=@n118.RGB, B=@n120), B=LinearInterpolate(A=@n123, B=@n124.G, Alpha=@n125))
  )
  Reroute(None=Add(A=Add(A=@n89, B=@n97), B=Transform(None=@n126) {transformSourceType=TRANSFORMSOURCE_Local}))
  ScalarParameter() {defaultValue=0.5, group=Dissolve, parameterName=Dissolve Dynamic Test}
 ]
}
